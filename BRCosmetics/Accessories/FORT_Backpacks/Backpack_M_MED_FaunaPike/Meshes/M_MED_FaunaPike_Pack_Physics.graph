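# Physics asset for the M_MED_FaunaPike backpack cosmetic, defining simulation bodies for two sword chains, front and back.
# Each chain is a series of capsules constrained link to link from an MSTR root: five links in back, eight in front.
# All joints lock twist, and most front joints alternate which swing axis is tighter, like interlocking chain links.
# A large, unconstrained pelvis capsule provides a body collision volume.
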PhysicsAsset /BRCosmetics/Accessories/FORT_Backpacks/Backpack_M_MED_FaunaPike/Meshes/M_MED_FaunaPike_Pack_Physics
bodies {
 MSTR_chain_back_sword [{shapeType=Capsule, length=0.505, radius=0.505, shapeName=MSTR_chain_back_sword_capsule, rotation=<rot 0 -90 0>}]
 MSTR_chain_front_sword [{shapeType=Capsule, length=0.505, radius=0.505, shapeName=MSTR_chain_front_sword_capsule, rotation=<rot 0 -90 0>}]
 dyn_chain_back_sword_01 [{shapeType=Capsule, length=2.6275, radius=1.7324, shapeName=dyn_chain_back_sword_01_capsule, center=<1.4 0 0>, rotation=<rot 90 0 0>}]
 dyn_chain_back_sword_02 [{shapeType=Capsule, length=2, radius=1, shapeName=dyn_chain_back_sword_02_capsule, center=<1 0 0>, rotation=<rot 90 0 0>}]
 dyn_chain_back_sword_03 [{shapeType=Capsule, length=2, radius=1, shapeName=dyn_chain_back_sword_03_capsule, center=<1 0 0>, rotation=<rot 90 0 0>}]
 dyn_chain_back_sword_04 [{shapeType=Capsule, length=2, radius=1, shapeName=dyn_chain_back_sword_04_capsule, center=<1 0 0>, rotation=<rot 90 0 0>}]
 dyn_chain_back_sword_05 [{shapeType=Capsule, length=2.9225, radius=2.3208, shapeName=dyn_chain_back_sword_05_capsule, center=<3.4253 0 0>, rotation=<rot 90 0 0>}]
 dyn_chain_front_sword_01 [{shapeType=Capsule, length=1.5209, radius=0.8316, shapeName=dyn_chain_front_sword_01_capsule, center=<1.4 0 0>, rotation=<rot 90 0 0>}]
 dyn_chain_front_sword_02 [{shapeType=Capsule, length=1, radius=1, shapeName=dyn_chain_front_sword_02_capsule, center=<1 0 0>, rotation=<rot 90 0 0>}]
 dyn_chain_front_sword_03 [{shapeType=Capsule, length=1, radius=1, shapeName=dyn_chain_front_sword_03_capsule, center=<1 0 0>, rotation=<rot 90 0 0>}]
 dyn_chain_front_sword_04 [{shapeType=Capsule, length=1, radius=1, shapeName=dyn_chain_front_sword_04_capsule, center=<1 0 0>, rotation=<rot 90 0 0>}]
 dyn_chain_front_sword_05 [{shapeType=Capsule, length=1, radius=1, shapeName=dyn_chain_front_sword_05_capsule, center=<1 0 0>, rotation=<rot 90 0 0>}]
 dyn_chain_front_sword_06 [{shapeType=Capsule, length=1, radius=1, shapeName=dyn_chain_front_sword_06_capsule, center=<1 0 0>, rotation=<rot 90 0 0>}]
 dyn_chain_front_sword_07 [{shapeType=Capsule, length=1, radius=1, shapeName=dyn_chain_front_sword_07_capsule, center=<1 0 0>, rotation=<rot 90 0 0>}]
 dyn_chain_front_sword_08 [{shapeType=Capsule, length=4.7143, radius=2.3251, shapeName=dyn_chain_front_sword_08_capsule, center=<3.3 0 0>, rotation=<rot 90 0 0>}]
 pelvis_collision [{shapeType=Capsule, length=14.7741, radius=16.0173, shapeName=pelvis_collision_capsule, center=<-5.969 0 0>}]
}
constraints [
 {bone1Name=dyn_chain_back_sword_01, bone2Name=MSTR_chain_back_sword, swing1LimitDegrees=0, swing2LimitDegrees=45, twistLimitDegrees=0}
 {bone1Name=dyn_chain_back_sword_02, bone2Name=dyn_chain_back_sword_01, swing1LimitDegrees=45, swing2LimitDegrees=45, twistLimitDegrees=0}
 {bone1Name=dyn_chain_back_sword_03, bone2Name=dyn_chain_back_sword_02, swing1LimitDegrees=45, swing2LimitDegrees=45, twistLimitDegrees=0}
 {bone1Name=dyn_chain_back_sword_04, bone2Name=dyn_chain_back_sword_03, swing1LimitDegrees=45, swing2LimitDegrees=45, twistLimitDegrees=0}
 {bone1Name=dyn_chain_back_sword_05, bone2Name=dyn_chain_back_sword_04, swing1LimitDegrees=45, swing2LimitDegrees=45, twistLimitDegrees=0}
 {bone1Name=dyn_chain_front_sword_01, bone2Name=MSTR_chain_front_sword, swing1LimitDegrees=75, swing2LimitDegrees=45, twistLimitDegrees=0}
 {bone1Name=dyn_chain_front_sword_02, bone2Name=dyn_chain_front_sword_01, swing1LimitDegrees=8, swing2LimitDegrees=45, twistLimitDegrees=0}
 {bone1Name=dyn_chain_front_sword_03, bone2Name=dyn_chain_front_sword_02, swing1LimitDegrees=45, swing2LimitDegrees=15, twistLimitDegrees=0}
 {bone1Name=dyn_chain_front_sword_04, bone2Name=dyn_chain_front_sword_03, swing1LimitDegrees=12, swing2LimitDegrees=45, twistLimitDegrees=0}
 {bone1Name=dyn_chain_front_sword_05, bone2Name=dyn_chain_front_sword_04, swing1LimitDegrees=45, swing2LimitDegrees=8, twistLimitDegrees=0}
 {bone1Name=dyn_chain_front_sword_06, bone2Name=dyn_chain_front_sword_05, swing1LimitDegrees=12, swing2LimitDegrees=45, twistLimitDegrees=0}
 {bone1Name=dyn_chain_front_sword_07, bone2Name=dyn_chain_front_sword_06, swing1LimitDegrees=45, swing2LimitDegrees=8, twistLimitDegrees=0}
 {bone1Name=dyn_chain_front_sword_08, bone2Name=dyn_chain_front_sword_07, swing1LimitDegrees=45, swing2LimitDegrees=45, twistLimitDegrees=0}
]
