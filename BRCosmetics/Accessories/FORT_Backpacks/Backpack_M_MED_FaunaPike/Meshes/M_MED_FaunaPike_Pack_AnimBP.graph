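# Animation Blueprint for the FaunaPike backpack, adding physics motion and corrective pack offsets as the character moves.
# Each update copies the parent body's pelvis world transform onto a pelvis_collision bone, feeding a RigidBody simulation based on the pack bone.
# A chain of additive pack bone transforms follows the simulation.
# The event graph also derives turn and lean alphas from spine_05 deltas against the pelvis, zeroing Jump Back in front end or skydive floating.
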AnimBlueprint /BRCosmetics/Accessories/FORT_Backpacks/Backpack_M_MED_FaunaPike/Meshes/M_MED_FaunaPike_Pack_AnimBP
target_skeleton=&/M_MED_FaunaPike_Pack_Skeleton
graphs [
 AnimGraph {nodes [
   n1 = "Transform (Modify) Bone - Bone: pelvis_collision"(
    ComponentPose=Local To Component(LocalPose=Input Pose())
    Alpha=1
    Translation=Get pelvis_location() -> (pelvis_location:Vector)
    Rotation=Get pelvis_rotation() -> (pelvis_rotation:Rotator)
    Scale="1,1,1"
   ) {
    rotationMode=BMM_Replace
    rotationSpace=BCS_WorldSpace
    scaleSpace=BCS_BoneSpace
    translationMode=BMM_Replace
    translationSpace=BCS_WorldSpace
    alphaBoolBlend {blendOption=Linear}
    alphaScaleBiasClamp {inRange {max=0, min=-40}, outRange {max=0, min=1}}
    boneToModify {boneName=pelvis_collision}
   }
   n2 = RigidBody(ComponentPose=@n1) {
    bClampLinearTranslationLimitToRefPose=true
    bOverrideWorldGravity=true
    overridePhysicsAsset=&/M_MED_FaunaPike_Pack_Physics
    baseBoneRef {boneName=pack}
    overrideWorldGravity {z=-980}
    simSpaceSettings {
     dampingAlpha=0
     maxAngularAcceleration=32
     maxAngularVelocity=8
     maxLinearAcceleration=1200
     maxLinearVelocity=700
     worldAlpha=1
     externalLinearDragV=<0.3 1.4 0.3>
    }
   }
   n3 = "Transform (Modify) Bone - Bone: pack"!(ComponentPose=@n2, bAlphaBoolEnabled:bool, Translation="0,-6,0", Rotation="-0.000,0.000,16.384", Scale) -> (Pose) {
    alphaInputType=Bool
    rotationMode=BMM_Additive
    translationMode=BMM_Additive
    boneToModify {boneName=pack}
    rotation {roll=16.384}
    translation {y=-6}
   }
   n4 = "Transform (Modify) Bone - Bone: pack"!(ComponentPose=@n3, Alpha:float, Translation="0,-6,0", Rotation="0,0,12", Scale) -> (Pose) {
    rotationMode=BMM_Additive
    translationMode=BMM_Additive
    alphaScaleBiasClamp {bInterpResult=true}
    boneToModify {boneName=pack}
    rotation {roll=12}
    translation {y=-6}
   }
   n5 = "Transform (Modify) Bone - Bone: pack"(ComponentPose=@n4, Translation="0,-6,0", Rotation="0,0,12") {
    rotationMode=BMM_Additive
    translationMode=BMM_Additive
    alphaScaleBiasClamp {bInterpResult=true}
    boneToModify {boneName=pack}
    rotation {roll=12}
    translation {y=-6}
   }
   n6 = "Transform (Modify) Bone - Bone: pack"(ComponentPose=@n5, Translation="0,0,10", Rotation="0,0,45") {
    rotationMode=BMM_Additive
    translationMode=BMM_Additive
    alphaScaleBiasClamp {bInterpResult=true, interpSpeedIncreasing=15}
    boneToModify {boneName=pack}
    rotation {roll=45}
    translation {z=10}
   }
   n7 = "Transform (Modify) Bone - Bone: pack"(ComponentPose=@n6, Translation="0,-5,7", Rotation="0,0,15") {
    rotationMode=BMM_Additive
    translationMode=BMM_Additive
    alphaScaleBiasClamp {bInterpResult=true}
    boneToModify {boneName=pack}
    rotation {roll=15}
    translation {y=-5, z=7}
   }
   Output Pose(Result=Component To Local(ComponentPose=@n7)) {layerGroup=None}
  ]}
 EventGraph {nodes [
   Event BlueprintUpdateAnimation()
   n1 = GetPartSkeletalMeshComponent()
   > Is Valid(InputObject=@n1)
   n2 = GetAttachParent!(self:Scene Component Object Reference=@n1) -> (ReturnValue:Scene Component Object Reference)
   > Is Valid(InputObject=@n2)
   > Set Body_Mesh(Body_Mesh:Scene Component Object Reference=@n2) -> (Output_Get:Scene Component Object Reference)
   > n3 = Sequence()
   n4 = BreakTransform(InTransform=GetSocketTransform(self=Get Body_Mesh() -> (Body_Mesh:Scene Component Object Reference), InSocketName=pelvis, TransformSpace=RTS_World))
   > Set pelvis_location(pelvis_location:Vector=@n4) -> (Output_Get:Vector)
   > Set pelvis_rotation(pelvis_rotation:Rotator=@n4.Rotation) -> (Output_Get:Rotator)
   n5 = GetOwnerSkeletalMeshComponent()
   Set Alpha Swing Turn Left(
    execute=@n3.then_1
    Alpha Swing Turn Left:float=GetAlphaFromReferenceTransformDelta(SkeletalMeshComponent=@n5, BoneName=spine_05, BaseName=pelvis, DeltaComponentType=QuaternionTwist, InputRange=(X=40,Y=55), TwistAxis="1,0,0")
   ) -> (Output_Get:float)
   > Set Alpha Swing Turn Right(Alpha Swing Turn Right:float=GetAlphaFromReferenceTransformDelta(
     SkeletalMeshComponent=@n5
     BoneName=spine_05
     BaseName=pelvis
     DeltaComponentType=QuaternionTwist
     InputRange=(X=-30,Y=-40)
     TwistAxis="1,0,0"
    )) -> (Output_Get:float)
   > Set Alpha Output(Alpha Output:float=GetAlphaFromReferenceTransformDelta(SkeletalMeshComponent=@n5, BoneName=spine_05, BaseName=pelvis, InputRange=(X=-35,Y=-45), TwistAxis="1,0,0")) -> (Output_Get:float)
   n6 = Get bIsInFrontEnd()
   n7 = GetAlphaFromReferenceTransformDelta(SkeletalMeshComponent=@n5, BoneName=spine_05, BaseName=pelvis, InputRange=(X=0,Y=-2), TwistAxis="1,0,0")
   n8 = NOT Boolean(A=Get bIsSkydiveFloating())
   > Set Jump Back Alpha(Jump Back Alpha:float=float * float(A=float * float(A=@n7, B=@n8), B:bool=NOT Boolean(A=@n6)) {is=title K2Node_PromotableOperator}) -> (Output_Get:float)
   TryGetPawnOwner()
  ]}
]
variables [
 Alpha Output {type=number}
 Alpha Swing Turn Left {type=number}
 Alpha Swing Turn Right {type=number}
 Body_Mesh {type=SceneComponent, default=null}
 Jump Back Alpha {type=number}
 pelvis_location {type=Vector}
 pelvis_rotation {type=Rotator}
]
